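# Unlit 2D material for the jump flooding tools that detects edges in a thresholded channel of a Seed texture.
# UVs are snapped to texel centers using TextureResolution, then fed with the threshold and channel mask into a Custom HLSL node.
# The node compares four axial neighbors, writing averaged half-texel boundary UVs to RG, or zero off edges, and the thresholded mask to alpha.
# That alpha is also inverted, and the material uses alpha-composite blending.
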Material /Volumetrics/Tools/JumpFlooding/Materials/Generation/M_DetectEdges_2D
graph {
 properties {blendMode=BLEND_AlphaComposite, pixelDepthOffsetMode=PDOM_Legacy, shadingModel=MSM_Unlit, usedShadingModels=Unlit}
 nodes [
  MaterialFunctionCall!(UVs, Texture Size) {materialFunction=/Engine/Functions/Engine_MaterialFunctions01/Texturing/PointSampledUVs}
  MaterialExpressionTextureObjectParameter_0 = TextureObjectParameter() {parameterName=Seed, samplerType=SAMPLERTYPE_LinearColor, texture=/Engine/MaterialTemplates/Textures/T_multiMask}
  n1 = ComponentMask(None=VectorParameter() {parameterName=TextureResolution, defaultValue=<rgba 256 256 0 0>}.RGB) {g=true, r=true}
  n2 = Multiply(A=TextureCoordinate(), B=@n1)
  MaterialExpressionAdd_0 = Add(A=Divide(A=Floor(None=@n2), B=@n1), B=Divide(A=Divide(B=@n1)))
  MaterialExpressionComponentMask_3 = ComponentMask(None=VectorParameter() {parameterName=TextureResolution, defaultValue=<rgba 256 256 0 0>}.RGB) {g=true, r=true}
  MaterialExpressionScalarParameter_0 = ScalarParameter() {defaultValue=0.1, parameterName=Thresh}
  n3 = VectorParameter() {parameterName=MaskChannel, defaultValue=<rgba 1 0 0 0>}
  MaterialExpressionAppendVector_0 = AppendVector(A=@n3.RGB, B=@n3.RGB)
  n4 = Custom!(
   Tex=@MaterialExpressionTextureObjectParameter_0
   UV=@MaterialExpressionAdd_0
   TextureSize=@MaterialExpressionComponentMask_3
   Threshold=@MaterialExpressionScalarParameter_0
   MaskChannel=@MaterialExpressionAppendVector_0
  ) {
   code="float texelsize = 1/TextureSize.x;\r\n\r\nfloat2 offsets[8] = {float2(1,0) , float2(-1,0) , float2(0,1) , float2 (0,-1), float2(1,1), float2(1,-1), float2(-1,1), float2(-1,-1)};\r\n\r\nfloat startsample =  dot( Tex.SampleLevel(TexSampler, UV, 0), MaskChannel);\r\n\r\nfloat startcheck = saturate(ceil(startsample - Threshold));\r\n\r\nint neighbors = 0;\r\nfloat3 accum = 0;\r\n\r\nint i = 0;\r\nwhile(i<4)\r\n{\r\n\tfloat2 offsetsample = dot( Tex.SampleLevel(TexSampler, UV + (offsets[i] * texelsize), 0), MaskChannel);\r\n\tfloat offsetcheck = saturate(ceil(offsetsample - Threshold));\r\n\r\n\tif(offsetcheck != startcheck &&&& startcheck <= 1)\r\n\t{\r\n\t\taccum.xy += offsets[i] * texelsize * 0.5;\r\n\t\tneighbors += 1;\r\n\t}\r\n\r\ni++;\r\n}\r\n\r\nif (neighbors > 0)\r\n{\r\n\taccum.xy /= neighbors;\r\n\taccum.xy += UV;\r\n}\r\n\r\nreturn float4( accum, startcheck);"
   description=Detect Edges
   outputType=CMOT_Float4
   inputs [
    {
     inputName=Tex
     input {expression=&self:MaterialExpressionTextureObjectParameter_0}
    }
    {inputName=UV, input {expression=&self:MaterialExpressionAdd_0}}
    {
     inputName=TextureSize
     input {expression=&self:MaterialExpressionComponentMask_3}
    }
    {
     inputName=Threshold
     input {expression=&self:MaterialExpressionScalarParameter_0}
    }
    {
     inputName=MaskChannel
     input {expression=&self:MaterialExpressionAppendVector_0}
    }
   ]
  }
  OneMinus(None=ComponentMask(None=@n4) {a=true})
 ]
}
