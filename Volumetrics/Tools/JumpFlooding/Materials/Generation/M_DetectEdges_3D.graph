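# Unlit alpha-composite material that detects 3D edges in a flipbook-packed volume texture, apparently to generate seeds for jump flooding.
# Point-sampled UVs are mapped to volume UVW for a Custom HLSL node that thresholds the selected Seed channel.
# The node compares each voxel with six face neighbors, finding the averaged half-voxel boundary position where they differ.
# It returns that position with the inside/outside mask, which is also inverted.
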Material /Volumetrics/Tools/JumpFlooding/Materials/Generation/M_DetectEdges_3D
graph {
 properties {blendMode=BLEND_AlphaComposite, pixelDepthOffsetMode=PDOM_Legacy, shadingModel=MSM_Unlit, usedShadingModels=Unlit}
 nodes [
  MaterialExpressionTextureObjectParameter_0 = TextureObjectParameter() {
   parameterName=Seed
   samplerType=SAMPLERTYPE_LinearColor
   texture=/Engine/Functions/Engine_MaterialFunctions02/PivotPainter2/Black_1x1_EXR_Texture
  }
  MaterialExpressionMaterialFunctionCall_3 = MaterialFunctionCall!(
   UVs=TextureCoordinate()
   Texture Size=ComponentMask(None=VectorParameter() {parameterName=TextureResolution, defaultValue=<rgba 4096 4096 0 0>}.RGB) {g=true, r=true}
  ) {materialFunction=/Engine/Functions/Engine_MaterialFunctions01/Texturing/PointSampledUVs}
  MaterialExpressionComponentMask_1 = ComponentMask(None=VectorParameter() {parameterName=XYFrames, defaultValue=<rgba 16 16 0 0>}.RGB) {g=true, r=true}
  MaterialExpressionMaterialFunctionCall_1 = MaterialFunctionCall!(UVs=@MaterialExpressionMaterialFunctionCall_3.Result, XY Frames=@MaterialExpressionComponentMask_1) {materialFunction=/Volumetrics/Content/VolumeTextures/MaterialFunctions/EncodeVolumeCoordinates}
  MaterialExpressionVectorParameter_3 = VectorParameter() {parameterName=VolumeResolution, defaultValue=<rgba 256 256 256 0>}
  MaterialExpressionScalarParameter_0 = ScalarParameter() {defaultValue=0.01, parameterName=Thresh}
  n1 = VectorParameter() {parameterName=MaskChannel, defaultValue=<rgba 0 0 0 1>}
  MaterialExpressionAppendVector_0 = AppendVector(A=@n1.RGB, B=@n1.RGB)
  n2 = Custom!(
   Tex=@MaterialExpressionTextureObjectParameter_0
   UV=@MaterialExpressionMaterialFunctionCall_3.Result
   UVW=@MaterialExpressionMaterialFunctionCall_1.Result
   XYFrames=@MaterialExpressionComponentMask_1
   VolumeResolution=@MaterialExpressionVectorParameter_3.RGB
   Threshold=@MaterialExpressionScalarParameter_0
   MaskChannel=@MaterialExpressionAppendVector_0
  ) {
   code="float3 voxelsize = 1/VolumeResolution;\r\n\r\nint neighbors = 0;\r\nfloat3 accum = 0;\r\n\r\nstruct EdgeFinder\r\n{\r\n\tTexture2D VolTex;\r\n\tSamplerState VolTexSampler;\r\n\tfloat2 xysize;\r\n\tfloat Mask;\r\n\tfloat numframes;\r\n\t\r\n\t\r\n\tfloat2 GetSubFrame(float xsize, float idx)\r\n\t{\r\n\t\tfloat2 xyidx = 0;\r\n\t\txyidx.y = floor(idx / xsize);\r\n\t\txyidx.x = idx - xsize * xyidx.y;\r\n\r\n\t\treturn xyidx;\r\n\t}\r\n\r\n\tfloat4 PseudoVolSample(float3 inPos)\r\n\t{\r\n\t\r\n\t\tfloat z = inPos.z; //don't need half texel offset since we basically round using this behavior\r\n\t\tfloat zframe = floor(z * numframes);\r\n\t\tfloat zphase = frac(z * numframes);\r\n\t\r\n\t\tfloat2 subuv = frac(inPos.xy) / xysize;\r\n\t\r\n\t\tfloat2 curframe = GetSubFrame(xysize.x, zframe) / xysize;\r\n\r\n\t\treturn VolTex.SampleLevel(VolTexSampler, subuv + curframe, 0);\r\n\t}\r\n\r\n\tfloat4 GenerateMask(float2 inUV, float3 inUVW, float3 inVoxelSize, float4 inMaskChannel, float inTreshold)\r\n\t{\r\n\t\t\r\n\t\tint numneighbors = 0;\r\n\t\tfloat3 accumpos = 0;\r\n\t\tMask = 0;\r\n\r\n\t\r\n\t\tfloat3 offsets[6] = {float3(1,0,0) , float3(-1,0,0) , float3(0,1,0) , float3(0,-1,0), float3(0,0,1), float3(0,0,-1)};\r\n\t\t\r\n\t\tfloat startsample =  dot( VolTex.SampleLevel(VolTexSampler, inUV, 0), inMaskChannel);\r\n\r\n\t\tfloat startcheck = saturate(ceil(startsample - inTreshold));\r\n\t\t\r\n\t\tint i = 0;\r\n\t\twhile(i<6)\r\n\t\t{\r\n\t\t\tfloat offsetsample = dot( PseudoVolSample(inUVW + (offsets[i] * inVoxelSize)), inMaskChannel);\r\n\t\t\tfloat offsetcheck = saturate(ceil(offsetsample - inTreshold));\r\n\r\n\t\t\tif(offsetcheck != startcheck &&&& startcheck <= 1)\r\n\t\t\t{\r\n\t\t\t\taccumpos.xyz += offsets[i] * inVoxelSize * 0.5;\r\n\t\t\t\tnumneighbors += 1;\r\n\t\t\t}\r\n\r\n\t\t\ti++;\r\n\t\t}\r\n\t\tMask = startcheck;\r\n\t\t\r\n\t\treturn float4(accumpos, numneighbors);\r\n\t}\r\n\r\n};\r\n\r\n\r\n\r\n\r\n\r\nEdgeFinder EdgeMask;\r\nEdgeMask.VolTex = Tex;\r\nEdgeMask.VolTexSampler = TexSampler;\r\nEdgeMask.xysize = XYFrames;\r\nEdgeMask.numframes = XYFrames.x * XYFrames.y;\r\n\r\n\r\nfloat4 edges = EdgeMask.GenerateMask(UV, UVW, voxelsize, MaskChannel, Threshold);\r\naccum = edges.xyz;\r\nneighbors = edges.a;\r\n\r\nif (neighbors > 0)\r\n{\r\n\taccum.xyz /= neighbors;\r\n\taccum.xyz += UVW;\r\n}\r\n\r\nreturn float4( accum, EdgeMask.Mask);\r\n\r\n\r\n"
   description=Detect Edges
   outputType=CMOT_Float4
   inputs [
    {
     inputName=Tex
     input {expression=&self:MaterialExpressionTextureObjectParameter_0}
    }
    {
     inputName=UV
     input {expression=&self:MaterialExpressionMaterialFunctionCall_3}
    }
    {
     inputName=UVW
     input {expression=&self:MaterialExpressionMaterialFunctionCall_1}
    }
    {
     inputName=XYFrames
     input {expression=&self:MaterialExpressionComponentMask_1}
    }
    {
     inputName=VolumeResolution
     input {
      expression=&self:MaterialExpressionVectorParameter_3
      mask=1
      maskB=1
      maskG=1
      maskR=1
     }
    }
    {
     inputName=Threshold
     input {expression=&self:MaterialExpressionScalarParameter_0}
    }
    {
     inputName=MaskChannel
     input {expression=&self:MaterialExpressionAppendVector_0}
    }
   ]
  }
  ComponentMask(None=@n2) {b=true, g=true, r=true}
  Constant() {r=1}
  OneMinus(None=ComponentMask(None=@n2) {a=true})
 ]
}
