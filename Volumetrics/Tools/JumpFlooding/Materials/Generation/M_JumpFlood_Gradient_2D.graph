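# Unlit alpha-composite material that derives a 2D gradient and signed distance from a jump flood render target.
# It point-samples the RT at resolution-snapped UVs and subtracts those UVs from the stored RG coordinate to get an offset.
# The RT alpha sets the sign through a lerp between 1 and -1.
# The signed offset is normalized into a direction, while its signed length is bias-scaled and inverted.
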Material /Volumetrics/Tools/JumpFlooding/Materials/Generation/M_JumpFlood_Gradient_2D
graph {
 properties {
  bAllowNegativeEmissiveColor=true
  blendMode=BLEND_AlphaComposite
  pixelDepthOffsetMode=PDOM_Legacy
  shadingModel=MSM_Unlit
  usedShadingModels=Unlit
 }
 nodes [
  n1 = MaterialFunctionCall!(
   UVs=TextureCoordinate()
   Texture Size=ComponentMask(None=VectorParameter() {parameterName=TextureResolution, defaultValue=<rgba 2048 1024 0 0>}.RGB) {g=true, r=true}
  ) {materialFunction=/Engine/Functions/Engine_MaterialFunctions01/Texturing/PointSampledUVs}
  n2 = TextureSampleParameter2D(UVs=@n1.Result) {
   parameterName=RT
   samplerType=SAMPLERTYPE_LinearColor
   texture=/Engine/Functions/Engine_MaterialFunctions02/PivotPainter2/Black_1x1_EXR_Texture
  }
  n3 = Subtract(A=ComponentMask(None=@n2.RGB) {g=true, r=true}, B=@n1.Result)
  n4 = LinearInterpolate(Alpha=@n2.A) {constA=1, constB=-1}
  Normalize(VectorInput=Multiply(A=@n3, B=@n4))
  n5 = Distance(A=Constant(), B=@n3)
  OneMinus(None=ConstantBiasScale(None=Multiply(A=@n5, B=@n4)))
 ]
}
