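# Unlit material that turns a 3D jump flood result, stored as tiled slices in a 2D render target, into gradient direction and signed distance.
# It point-samples the target and subtracts each texel's encoded volume coordinate from the sampled RGB position.
# Sampled alpha sets the sign of that offset and its length.
# The offset is normalized when non-zero, and the signed length is bias-scaled and inverted.
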Material /Volumetrics/Tools/JumpFlooding/Materials/Generation/M_JumpFlood_Gradient_3D
graph {
 properties {
  bAllowNegativeEmissiveColor=true
  blendMode=BLEND_AlphaComposite
  pixelDepthOffsetMode=PDOM_Legacy
  shadingModel=MSM_Unlit
  usedShadingModels=Unlit
 }
 nodes [
  n1 = MaterialFunctionCall!(
   UVs=TextureCoordinate()
   Texture Size=ComponentMask(None=VectorParameter() {parameterName=TextureResolution, defaultValue=<rgba 4096 4096 0 0>}.RGB) {g=true, r=true}
  ) {materialFunction=/Engine/Functions/Engine_MaterialFunctions01/Texturing/PointSampledUVs}
  n2 = TextureSampleParameter2D(UVs=@n1.Result) {
   parameterName=RT
   samplerType=SAMPLERTYPE_LinearColor
   texture=/Engine/Functions/Engine_MaterialFunctions02/PivotPainter2/Black_1x1_EXR_Texture
  }
  n3 = VectorParameter() {parameterName=XYFrames, defaultValue=<rgba 16 16 0 0>}
  n4 = Subtract(
   A=@n2.RGB
   B=MaterialFunctionCall!(UVs=@n1.Result, XY Frames=ComponentMask(None=@n3.RGB) {g=true, r=true}) {materialFunction=/Volumetrics/Content/VolumeTextures/MaterialFunctions/EncodeVolumeCoordinates}.Result
  )
  n5 = LinearInterpolate(Alpha=@n2.A) {constA=1, constB=-1}
  n6 = Multiply(A=@n4, B=@n5)
  If(A=DotProduct(A=Abs(None=@n6), B=Constant() {r=1}), "A > B"=Normalize(VectorInput=@n6), "A == B"=@n6, "A < B"=@n6)
  n7 = Distance(A=Constant(), B=@n4)
  OneMinus(None=ConstantBiasScale(None=Multiply(A=@n7, B=@n5)))
 ]
}
