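# Unlit jump-flooding tool material that computes a blurred 3D gradient of a volume, such as a distance field, stored as a tiled 2D render target.
# A Custom node averages six-neighbour axis differences over each voxel and its six neighbours, clamped at volume bounds.
# Show Edges blends the source toward a second Voronoi texture, and Strength scales the result.
# Side branches compute a faded-edge gradient and signed direction and distance from RT.
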Material /Volumetrics/Tools/JumpFlooding/Materials/Generation/M_JumpFlood_Gradient_3D_Blurred
graph {
 properties {
  bAllowNegativeEmissiveColor=true
  blendMode=BLEND_AlphaComposite
  pixelDepthOffsetMode=PDOM_Legacy
  shadingModel=MSM_Unlit
  usedShadingModels=Unlit
 }
 nodes [
  MaterialExpressionTextureObjectParameter_0 = TextureObjectParameter() {parameterName=RT, texture=/Volumetrics/Tools/JumpFlooding/Textures/RT_JumpFlood_3D_4k_DF}
  MaterialExpressionMaterialFunctionCall_0 = MaterialFunctionCall!(
   UVs=TextureCoordinate()
   Texture Size=ComponentMask(None=VectorParameter() {parameterName=TextureResolution, defaultValue=<rgba 2048 1024 0 0>}.RGB) {g=true, r=true}
  ) {materialFunction=/Engine/Functions/Engine_MaterialFunctions01/Texturing/PointSampledUVs}
  MaterialExpressionMaterialFunctionCall_8 = MaterialFunctionCall!(
   UVs=@MaterialExpressionMaterialFunctionCall_0.Result
   XY Frames=ComponentMask(None=VectorParameter() {parameterName=XYFrames, defaultValue=<rgba 16 8 0 0>}.RGB) {g=true, r=true}
  ) {materialFunction=/Volumetrics/Content/VolumeTextures/MaterialFunctions/EncodeVolumeCoordinates}
  MaterialExpressionVectorParameter_32 = VectorParameter() {parameterName=VolumeResolution, defaultValue=<rgba 128 128 128 0>}
  MaterialExpressionComponentMask_0 = ComponentMask(None=VectorParameter() {parameterName=XYFrames, defaultValue=<rgba 16 8 0 0>}.RGB) {g=true, r=true}
  MaterialExpressionTextureObjectParameter_2 = TextureObjectParameter() {parameterName=RT_1, texture=null}
  MaterialExpressionScalarParameter_12 = ScalarParameter() {parameterName=Show Edges}
  n1 = Custom!(
   Tex=@MaterialExpressionTextureObjectParameter_0
   UV0=@MaterialExpressionMaterialFunctionCall_0.Result
   UVW=@MaterialExpressionMaterialFunctionCall_8.Result
   TextureSize=@MaterialExpressionVectorParameter_32.RGB
   XYFrames=@MaterialExpressionComponentMask_0
   Tex2=@MaterialExpressionTextureObjectParameter_2
   ShowEdges=@MaterialExpressionScalarParameter_12
  ) {
   code="const float NumFrames = XYFrames.x * XYFrames.y;\r\n\r\nfloat3 PixelSize = 1 / TextureSize;\r\n\r\nfloat3 offsets[7] = {float3(1,0,0), float3(-1,0,0), float3(0,1,0), float3(0,-1,0), float3(0,0,1), float3(0,0,-1), float3(0,0,0)};\r\n\r\nfloat3 accum = 0;\r\n\r\nint i = 0;\r\nwhile ( i < 7)\r\n{\r\n\t\r\n\tfloat3 CurUVW = UVW + offsets[i] * PixelSize;\r\n\tCurUVW = clamp(CurUVW, PixelSize * 0.5, 1 - (PixelSize * 0.5));\r\n\r\n\tfloat3 cursample = PseudoVolumeTexture(Tex, TexSampler, CurUVW, XYFrames, NumFrames).rgb;\r\n\tfloat3 curvoronoi = PseudoVolumeTexture(Tex2, Tex2Sampler, CurUVW, XYFrames, NumFrames).rgb;\r\n\r\n\tcursample = lerp(cursample, curvoronoi, ShowEdges);\r\n\r\n\tint j = 0;\r\n\twhile ( j < 6)\r\n\t{\r\n\t\tfloat3 SubUVW = CurUVW + offsets[j] * PixelSize;\r\n\t\tSubUVW = clamp(SubUVW, PixelSize * 0.5, 1 - (PixelSize * 0.5));\r\n\t\tfloat3 subsample = PseudoVolumeTexture(Tex, TexSampler, SubUVW, XYFrames, NumFrames).rgb;\r\n\t\tfloat3 subvoronoi = PseudoVolumeTexture(Tex2, Tex2Sampler, SubUVW, XYFrames, NumFrames).rgb;\r\n\r\n\t\tsubsample = lerp(subsample, subvoronoi, ShowEdges);\r\n\r\n\t\taccum += (subsample - cursample) * offsets[j];\r\n\t\tj++;\r\n\t}\r\n\r\ni++;\r\n}\r\n\r\nreturn accum / 7;\r\n"
   description=Gradient - 6x6 offsets
   inputs [
    {
     inputName=Tex
     input {expression=&self:MaterialExpressionTextureObjectParameter_0}
    }
    {
     inputName=UV0
     input {expression=&self:MaterialExpressionMaterialFunctionCall_0}
    }
    {
     inputName=UVW
     input {expression=&self:MaterialExpressionMaterialFunctionCall_8}
    }
    {
     inputName=TextureSize
     input {
      expression=&self:MaterialExpressionVectorParameter_32
      mask=1
      maskB=1
      maskG=1
      maskR=1
     }
    }
    {
     inputName=XYFrames
     input {expression=&self:MaterialExpressionComponentMask_0}
    }
    {
     inputName=Tex2
     input {expression=&self:MaterialExpressionTextureObjectParameter_2}
    }
    {
     inputName=ShowEdges
     input {expression=&self:MaterialExpressionScalarParameter_12}
    }
   ]
  }
  n2 = Multiply(A=@n1, B=ScalarParameter() {defaultValue=8, parameterName=Strength})
  LinearInterpolate(B=@n2, Alpha=ScalarParameter() {defaultValue=1, parameterName=456})
  n3 = ComponentMask(None=VectorParameter() {parameterName=XYFrames, defaultValue=<rgba 16 8 0 0>}.RGB) {g=true, r=true}
  Multiply(
   A=MaterialFunctionCall!(
    UVW=MaterialFunctionCall(XY Frames=@n3) {materialFunction=&EncodeVolumeCoordinates}.Result
    Falloff Width=ScalarParameter() {defaultValue=0.05, parameterName=FW}
    Hardness=ScalarParameter() {parameterName=hardness}
    Falloff Power=ScalarParameter() {defaultValue=2, parameterName=FP}
    XY Frames=ScalarParameter() {defaultValue=10, parameterName=XYFrames}
   ) {materialFunction=/Volumetrics/Content/VolumeTextures/MaterialFunctions/VolumeEdgeMask}.Result
   B=@n2
  )
  n4 = TextureSampleParameter2D(UVs=@MaterialExpressionMaterialFunctionCall_0.Result) {parameterName=RT, samplerType=SAMPLERTYPE_LinearColor, texture=/Volumetrics/Content/VolumeTextures/Textures/RT_Default}
  n5 = Subtract(A=@n4.RGB, B=@MaterialExpressionMaterialFunctionCall_8.Result)
  n6 = LinearInterpolate(Alpha=@n4.A) {constA=1, constB=-1}
  Normalize(VectorInput=Multiply(A=@n5, B=@n6))
  n7 = Distance(A=Constant(), B=@n5)
  OneMinus(None=ConstantBiasScale(None=Multiply(A=@n7, B=@n6)))
  OneMinus(None=TextureSample() {texture=&RT_JumpFlood_3D_4k_DF}.R)
 ]
}
