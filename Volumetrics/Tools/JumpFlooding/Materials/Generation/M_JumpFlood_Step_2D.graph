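# Unlit alpha-composite material that runs one 2D jump flood step over the RT texture, propagating nearest seed coordinates.
# A Custom node point-samples eight neighbours, keeping the closest non-empty seed, with a Tiling parameter correcting wrapped coordinates.
# The jump distance derives from TextureResolution's X and halves as Index rises, passing through below one pixel.
# The starting sample's alpha is carried through, then inverted after masking.
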Material /Volumetrics/Tools/JumpFlooding/Materials/Generation/M_JumpFlood_Step_2D
graph {
 properties {
  bAllowNegativeEmissiveColor=true
  blendMode=BLEND_AlphaComposite
  pixelDepthOffsetMode=PDOM_Legacy
  shadingModel=MSM_Unlit
  usedShadingModels=Unlit
 }
 nodes [
  MaterialExpressionTextureObjectParameter_0 = TextureObjectParameter() {parameterName=RT, texture=/Volumetrics/Tools/JumpFlooding/Textures/RT_JumpFlood_01}
  n1 = VectorParameter() {parameterName=TextureResolution, defaultValue=<rgba 64 64 0 0>}
  MaterialExpressionComponentMask_8 = ComponentMask(None=@n1.RGB) {g=true, r=true}
  MaterialExpressionMaterialFunctionCall_1 = MaterialFunctionCall!(UVs=TextureCoordinate(), Texture Size=@MaterialExpressionComponentMask_8) {materialFunction=/Engine/Functions/Engine_MaterialFunctions01/Texturing/PointSampledUVs}
  n2 = Add(A=ScalarParameter() {parameterName=Index})
  MaterialExpressionMultiply_1 = Multiply(A=Max(A=@n1.R, B=@n1.R) {desc=Size}, B=Power(Base=Constant() {r=2}, Exp=Multiply(A=@n2) {constB=-1}))
  MaterialExpressionScalarParameter_0 = ScalarParameter() {parameterName=Tiling}
  n3 = Custom!(
   Tex=@MaterialExpressionTextureObjectParameter_0
   UV0=@MaterialExpressionMaterialFunctionCall_1.Result
   TextureSize=@MaterialExpressionComponentMask_8
   offsetmult=@MaterialExpressionMultiply_1
   Tiling=@MaterialExpressionScalarParameter_0
  ) {
   code="##define QualityMode 1\r\n\r\nfloat2 PixelSize = 1 / TextureSize;\r\nfloat2 offsets[8] = {float2(1,0) , float2(-1,0) , float2(0,1) , float2 (0,-1), float2(1,1), float2(1,-1), float2(-1,1), float2(-1,-1)};\r\n\r\n\r\nfloat4 startsample = Tex.SampleLevel(TexSampler, UV0, 0);\r\nfloat curlen = length(startsample.xy - UV0);\r\n\r\nfloat minlen = curlen;\r\nif (startsample.x == 0 &&&& startsample.y == 0) minlen = 10000000;\r\n\r\nfloat3 nearest = startsample.rgb;\r\n\r\nfloat3 cursample = 0;\r\n\r\nconst int samples = lerp( 4, 8, QualityMode);\r\n\r\nfor (int i = 0; i < samples; i++)\r\n{\r\n\r\n\tif (offsetmult < 1) \r\n\t{\r\n\t\treturn  float4(Tex.SampleLevel(TexSampler, UV0, 0).rgb, startsample.a);\r\n\t}\r\n\t\r\n\tfloat2 CurUV = UV0 + offsets[i] * offsetmult * PixelSize;\r\n\r\n\tcursample = Tex.SampleLevel(TexSampler, CurUV, 0);\r\n\t\r\n\tif (cursample.x == 0 &&&& cursample.y == 0) continue;\r\n\t\r\n\tcursample.xy += floor(CurUV) * Tiling;\r\n\r\n\tif (1)\r\n\t{\r\n\r\n\t\tcurlen = length(cursample.xy - UV0);\r\n\t\tif ( curlen < minlen )\r\n\t\t{\r\n\t\t\tminlen = curlen;\r\n\t\t\tnearest = cursample;\r\n\r\n\t\t}\r\n\t}\r\n\r\n}\r\n\r\nreturn float4(nearest, startsample.a);"
   description=Jump Step
   outputType=CMOT_Float4
   inputs [
    {
     inputName=Tex
     input {expression=&self:MaterialExpressionTextureObjectParameter_0}
    }
    {
     inputName=UV0
     input {expression=&self:MaterialExpressionMaterialFunctionCall_1}
    }
    {
     inputName=TextureSize
     input {expression=&self:MaterialExpressionComponentMask_8}
    }
    {
     inputName=offsetmult
     input {expression=&self:MaterialExpressionMultiply_1}
    }
    {
     inputName=Tiling
     input {expression=&self:MaterialExpressionScalarParameter_0}
    }
   ]
  }
  ComponentMask(None=@n3) {b=true, g=true, r=true}
  OneMinus(None=ComponentMask(None=@n3) {a=true})
 ]
}
