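# Unlit alpha-composite material that runs one 3D jump flooding step on a volume stored as a 2D grid of slices.
# Point-sampled UVs become volume coordinates; step size halves per Index, starting from the larger of VolumeResolution R and B (G is ignored).
# A custom HLSL node checks 26 neighbours at that step, keeping the nearest non-zero stored seed position.
# Steps below one voxel pass the texel through; a cheaper 10-neighbour branch is compiled out.
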Material /Volumetrics/Tools/JumpFlooding/Materials/Generation/M_JumpFlood_Step_3D
graph {
 properties {blendMode=BLEND_AlphaComposite, pixelDepthOffsetMode=PDOM_Legacy, shadingModel=MSM_Unlit, usedShadingModels=Unlit}
 nodes [
  ComponentMask() {g=true, r=true}
  MaterialExpressionTextureObjectParameter_0 = TextureObjectParameter() {parameterName=RT, texture=/Volumetrics/Tools/JumpFlooding/Textures/RT_JumpFlood_02}
  MaterialExpressionMaterialFunctionCall_0 = MaterialFunctionCall!(
   UVs=TextureCoordinate()
   Texture Size=ComponentMask(None=VectorParameter() {parameterName=TextureResolution, defaultValue=<rgba 4096 4096 0 0>}.RGB) {g=true, r=true}
  ) {materialFunction=/Engine/Functions/Engine_MaterialFunctions01/Texturing/PointSampledUVs}
  MaterialExpressionMaterialFunctionCall_8 = MaterialFunctionCall!(
   UVs=@MaterialExpressionMaterialFunctionCall_0.Result
   XY Frames=ComponentMask(None=VectorParameter() {parameterName=XYFrames, defaultValue=<rgba 16 8 0 0>}.RGB) {g=true, r=true}
  ) {materialFunction=/Volumetrics/Content/VolumeTextures/MaterialFunctions/EncodeVolumeCoordinates}
  MaterialExpressionVectorParameter_32 = VectorParameter() {parameterName=VolumeResolution, defaultValue=<rgba 128 128 128 0>}
  n1 = Add(A=ScalarParameter() {parameterName=Index})
  MaterialExpressionMultiply_1 = Multiply(
   A=Max(
    A=Max(A=@MaterialExpressionVectorParameter_32.R, B=@MaterialExpressionVectorParameter_32.R) {desc=Size}
    B=@MaterialExpressionVectorParameter_32.B
   ) {desc=Size}
   B=Power(Base=Constant() {r=2}, Exp=Multiply(A=@n1) {constB=-1})
  )
  MaterialExpressionComponentMask_10 = ComponentMask(None=VectorParameter() {parameterName=XYFrames, defaultValue=<rgba 16 8 0 0>}.RGB) {g=true, r=true}
  Custom!(
   Tex=@MaterialExpressionTextureObjectParameter_0
   UV0=@MaterialExpressionMaterialFunctionCall_0.Result
   UVW=@MaterialExpressionMaterialFunctionCall_8.Result
   TextureSize=@MaterialExpressionVectorParameter_32.RGB
   offsetmult=@MaterialExpressionMultiply_1
   XYFrames=@MaterialExpressionComponentMask_10
  ) {
   code="##define QualityMode 1\r\n\r\nconst float NumFrames = XYFrames.x * XYFrames.y;\r\n\r\nfloat3 PixelSize = 1 / TextureSize;\r\n\r\nstruct VolumeSample\r\n{\r\n\tfloat4 GetVolumeSample(float3 inPos, float2 inXYFrames, float inNumFrames, Texture2D inTex, SamplerState inTexSampler)\r\n\t{\r\n\t\r\n\tfloat zframe = round(inPos.z * inNumFrames);\r\n\r\n\tfloat2 uv = frac(inPos.xy) / inXYFrames;\r\n\r\n\tfloat2 xyidx = 0;\r\n\r\n\txyidx.y = floor(zframe / inXYFrames);\r\n\txyidx.x = zframe - inXYFrames * xyidx.y;\r\n\tfloat2 curframe = xyidx / inXYFrames;\r\n\r\n\txyidx.y = floor((zframe + 1) / inXYFrames.x);\r\n\txyidx.x = zframe + 1 - (inXYFrames.x * xyidx.y);\r\n\r\n\tfloat4\tsampleA = inTex.SampleLevel(inTexSampler, (uv + curframe) * 1, 0);\r\n\r\n\r\n\r\n\treturn sampleA;\r\n\t}\r\n\r\n};\r\n\r\nfloat3 offsets[26] = {float3(0,0,1), float3(1,0,1) , float3(-1,0,1) , float3(0,1,1) , float3(0,-1,1), float3(1,1,1), float3(1,-1,1), float3(-1,1,1), float3(-1,-1,1),\r\nfloat3(1,0,0) , float3(-1,0,0) , float3(0,1,0) , float3(0,-1,0), float3(1,1,0), float3(1,-1,0), float3(-1,1,0), float3(-1,-1,0),\r\nfloat3(0,0,-1), float3(1,0,-1) , float3(-1,0,-1) , float3(0,1,-1) , float3(0,-1,-1), float3(1,1,-1), float3(1,-1,-1), float3(-1,1,-1), float3(-1,-1,-1)};\r\n\r\nfloat3 simpleoffsets[10] = {float3(1,0,0), float3(-1,0,0), float3(0,1,0), float3(0,-1,0), float3(0,0,1), float3(0,0,-1), float3(1,1,0), float3(1,-1,0), float3(-1,1,0), float3(-1,-1,0)};\r\n\r\n//return  PseudoVolumeTexture(Tex, TexSampler, UVW, XYFrames, NumFrames).rgb;\r\n\r\nif (offsetmult < 1) return  Tex.SampleLevel(TexSampler, UV0, 0).rgb;\r\n\r\nfloat3 cursample = Tex.SampleLevel(TexSampler, UV0, 0).rgb;\r\n\r\nfloat curlen = length(cursample - UVW);\r\nfloat minlen = curlen;\r\n\r\nif (cursample.x == 0 &&&& cursample.y == 0 &&&& cursample.z == 0) minlen = 1000000;\r\n\r\nfloat3 nearest = cursample;\r\n\r\n##if QualityMode\r\nint i = 0;\r\nwhile ( i < 26)\r\n{\r\n\t\r\n\tfloat3 CurUVW = UVW + offsets[i] * offsetmult * PixelSize;\r\n\tVolumeSample CurrentVolSample;\r\n\t\r\n\tcursample = CurrentVolSample.GetVolumeSample(CurUVW, XYFrames, NumFrames, Tex, TexSampler).rgb;\r\n\r\n\r\n\tif (cursample.x > 0 || cursample.y > 0 || cursample.z > 0)\r\n\t{\r\n\r\n\t\tcurlen = length( (cursample - UVW) );\r\n\t\tif ( curlen < minlen )\r\n\t\t{\r\n\t\t\tminlen = curlen;\r\n\t\t\tnearest = cursample;\r\n\t\t}\r\n\t}\r\n\r\ni++;\r\n}\r\n##else\r\nint i = 0;\r\nwhile ( i < 10)\r\n{\r\n\t\r\n\tfloat3 CurUVW = UVW + simpleoffsets[i] * offsetmult * PixelSize;\r\n\tcursample = PseudoVolumeTexture(Tex, TexSampler, CurUVW, XYFrames, NumFrames).rgb;\r\n\r\n\r\n\tif (cursample.x > 0 || cursample.y > 0 || cursample.z > 0)\r\n\t{\r\n\r\n\t\tcurlen = length( (cursample - UVW) );\r\n\t\tif ( curlen < minlen )\r\n\t\t{\r\n\t\t\tminlen = curlen;\r\n\t\t\tnearest = cursample;\r\n\t\t}\r\n\t}\r\n\r\ni++;\r\n}\r\n##endif\r\n\r\nreturn nearest;"
   description=Jump Step
   inputs [
    {
     inputName=Tex
     input {expression=&self:MaterialExpressionTextureObjectParameter_0}
    }
    {
     inputName=UV0
     input {expression=&self:MaterialExpressionMaterialFunctionCall_0}
    }
    {
     inputName=UVW
     input {expression=&self:MaterialExpressionMaterialFunctionCall_8}
    }
    {
     inputName=TextureSize
     input {
      expression=&self:MaterialExpressionVectorParameter_32
      mask=1
      maskB=1
      maskG=1
      maskR=1
     }
    }
    {
     inputName=offsetmult
     input {expression=&self:MaterialExpressionMultiply_1}
    }
    {
     inputName=XYFrames
     input {expression=&self:MaterialExpressionComponentMask_10}
    }
   ]
  }
  OneMinus(None=TextureSampleParameter2D() {parameterName=RT, texture=/Volumetrics/Tools/JumpFlooding/Textures/RT_JumpFlood_01}.A)
 ]
}
